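# Sound cue for the UI sound of selecting an umbrella handle in frontend customization.
# A mixer layers three pitch-randomized branches: a Buildabrella swap, a heavily pitched-up challenge table hover, and a random pick of two Skye star sounds.
# The swap layer is weighted loudest and the stars quietest.
# Playback uses a non-critical in-game UI sound class and a shared umbrella-select attenuation asset.
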SoundCue /BRCosmetics/Sounds/Other/Fort_UI_Sounds/S13_Frontend/CustomizeSelect_Umbrella_Handle_Cue
root=Mixer_0
nodes [
 {id=Mixer_0, children [Modulator_1,Modulator_0,Modulator_3], properties {inputVolume [1.4,1,0.7]}}
 {id=Modulator_0, children [WavePlayer_2], properties {pitchMax=1.7, pitchMin=1.65}}
 {id=Modulator_1, children [WavePlayer_0], properties {pitchMax=1.15, pitchMin=1.1}}
 {id=Modulator_3, children [Random_0], properties {pitchMax=1.02, pitchMin=0.99}}
 {id=Random_0, children [WavePlayer_3,WavePlayer_4], properties {weights [1,1]}}
 {id=WavePlayer_0, properties {soundWaveAssetPtr=&/S13_Buildabrella_Swap}}
 {id=WavePlayer_2, properties {soundWaveAssetPtr=/Game/Sounds/Fort_UI_Sounds/S13_Frontend/S13_Challenge_Table_Hover}}
 {id=WavePlayer_3, properties {soundWaveAssetPtr=/BRCosmetics/Sounds/Other/Fort_UI_Sounds/S12_Frontend/S12_Skye_Stars_01}}
 {id=WavePlayer_4, properties {soundWaveAssetPtr=/Game/Sounds/Fort_UI_Sounds/S12_Frontend/S12_Skye_Stars_02}}
]
properties {
 attenuationSettings=&/CustomizeSelect_Umbrella_Att
 maxDistance=500
 soundClassObject=&Fort_UI_InGame_NonCritical
 volumeMultiplier=0.6
 attenuationOverrides {falloffDistance=4100, attenuationShapeExtents {x=900}}
 concurrencyOverrides {resolutionRule=PreventNew}
}
