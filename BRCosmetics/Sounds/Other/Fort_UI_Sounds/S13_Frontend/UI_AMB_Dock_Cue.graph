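# Sound cue for a dock water ambience, filed under the Season 13 frontend UI sounds.
# A mixer layers a quieter looping water stream bed with an endlessly repeating branch of metal knock-and-splash one-shots.
# That branch picks randomly among five equally weighted waves, varies volume slightly and inserts a short random delay between hits.
# Overridden attenuation leaves it non-spatialized, with distance falloff and occlusion filtering.
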SoundCue /BRCosmetics/Sounds/Other/Fort_UI_Sounds/S13_Frontend/UI_AMB_Dock_Cue
root=Mixer_0
nodes [
 {id=Delay_0, children [Modulator_0], properties {delayMax=3, delayMin=1}}
 {id=Looping_0, children [Delay_0]}
 {id=Mixer_0, children [WavePlayer_8,Looping_0], properties {inputVolume [0.4,1]}}
 {id=Modulator_0, children [Random_0], properties {volumeMax=1.1, volumeMin=0.98}}
 {id=Random_0, children [WavePlayer_0,WavePlayer_2,WavePlayer_3,WavePlayer_4,WavePlayer_5], properties {weights [1,1,1,1,1]}}
 {
  id=WavePlayer_0
  properties {soundWaveAssetPtr=/BRCosmetics/Sounds/Other/Apollo/Props/Fortillas/Amb_Metal_Water_Floatilla_Knock_Splash_01_01}
 }
 {
  id=WavePlayer_2
  properties {soundWaveAssetPtr=/BRCosmetics/Sounds/Other/Apollo/Props/Fortillas/Amb_Metal_Water_Floatilla_Knock_Splash_01_02}
 }
 {
  id=WavePlayer_3
  properties {soundWaveAssetPtr=/BRCosmetics/Sounds/Other/Apollo/Props/Fortillas/Amb_Metal_Water_Floatilla_Knock_Splash_01_03}
 }
 {
  id=WavePlayer_4
  properties {soundWaveAssetPtr=/BRCosmetics/Sounds/Other/Apollo/Props/Fortillas/Amb_Metal_Water_Floatilla_Knock_Splash_01_04}
 }
 {
  id=WavePlayer_5
  properties {soundWaveAssetPtr=/BRCosmetics/Sounds/Other/Apollo/Props/Fortillas/Amb_Metal_Water_Floatilla_Knock_Splash_01_05}
 }
 {id=WavePlayer_8, properties {bLooping=true, soundWaveAssetPtr=/Game/Athena/Sounds/Water/Ambient/Amb_Water_Stream_Med_Close_Loop_02}}
]
properties {
 bOverrideAttenuation=true
 maxDistance=800
 soundClassObject=&Fort_UI_InGame_Critical
 volumeMultiplier=0.6
 attenuationOverrides {
  bEnableOcclusion=true
  bSpatialize=false
  falloffDistance=300
  occlusionLowPassFilterFrequency=8000
  occlusionTraceChannel=ECC_GameTraceChannel11
  occlusionVolumeAttenuation=0.5
  stereoSpread=500
  attenuationShapeExtents {x=500}
 }
 concurrencyOverrides {maxCount=1, resolutionRule=PreventNew}
}
