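# UI_AMB_Fan_Cue is a SoundCue that plays a looping fan ambience, apparently for the Season 13 frontend, from the BuildaBrella fan loop wave.
# A single looping wave player feeds the output at boosted volume on the in-game critical UI sound class.
# Overridden attenuation makes it positional, with HRTF spatialization, short-range falloff and low-pass filtering when occluded.
# Concurrency settings allow one instance and block new ones.
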SoundCue /BRCosmetics/Sounds/Other/Fort_UI_Sounds/S13_Frontend/UI_AMB_Fan_Cue
root=WavePlayer_1
nodes [{id=WavePlayer_1, properties {bLooping=true, soundWaveAssetPtr=/Game/Sounds/Fort_UI_Sounds/S13_Frontend/S13_BuildaBrella_Fan_Loop}}]
properties {
 bOverrideAttenuation=true
 maxDistance=800
 soundClassObject=&Fort_UI_InGame_Critical
 volumeMultiplier=1.5
 attenuationOverrides {
  bEnableOcclusion=true
  falloffDistance=600
  occlusionLowPassFilterFrequency=1000
  occlusionTraceChannel=ECC_GameTraceChannel11
  spatializationAlgorithm=SPATIALIZATION_HRTF
  attenuationShapeExtents {x=200}
 }
 concurrencyOverrides {maxCount=1, resolutionRule=PreventNew}
}
