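# Season 13 frontend main-room ambience cue, a non-spatialized coastal and dock soundscape limited to one instance.
# A mixer layers a continuous bed with looped random wave one-shots, separated by random delays and varied in pitch and volume.
# The bed blends dock biome and two offshore ocean loops.
# A quality-level node plays quad versions on its first two levels and the alternate, mostly stereo-suffixed versions on the last two.
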SoundCue /BRCosmetics/Sounds/Other/Fort_UI_Sounds/S13_Frontend/UI_AMB_MainRoom_S13_Cue
root=Mixer_0
nodes [
 {id=Delay_0, children [Modulator_0], properties {delayMax=20, delayMin=5}}
 {id=Looping_0, children [Delay_0]}
 {id=Mixer_0, children [QualityLevel_0, Looping_0], properties {inputVolume [1,1.2]}}
 {id=Mixer_1, children [WavePlayer_16, WavePlayer_13, WavePlayer_18], properties {inputVolume [0.6,0.75,0.65]}}
 {id=Mixer_2, children [WavePlayer_14, WavePlayer_12, WavePlayer_15], properties {inputVolume [0.65,0.6,0.75]}}
 {id=Modulator_0, children [Random_1], properties {pitchMax=1.1, pitchMin=0.9, volumeMax=1.1, volumeMin=0.98}}
 {id=QualityLevel_0, children [Mixer_1,Mixer_1,Mixer_2,Mixer_2]}
 {id=Random_1, children [WavePlayer_3,WavePlayer_6,WavePlayer_7,WavePlayer_8,WavePlayer_9], properties {weights [1,1,1,1,1]}}
 {id=WavePlayer_12, properties {bLooping=true, soundWaveAssetPtr=/Game/Sounds/Apollo/Biomes/DirtyDocks/Biome_DirtyDocks_Day}}
 {
  id=WavePlayer_13
  properties {bLooping=true, soundWaveAssetPtr=/Game/Athena/Sounds/Water/Ambient/Amb_Water_Ocean_Offshore_Quad_Med_Dist_Loop_01}
 }
 {id=WavePlayer_14, properties {bLooping=true, soundWaveAssetPtr=/Game/Athena/Sounds/Water/Ambient/Amb_Water_Ocean_Offshore_Quad_Loop_01_st}}
 {
  id=WavePlayer_15
  properties {bLooping=true, soundWaveAssetPtr=/Game/Athena/Sounds/Water/Ambient/Amb_Water_Ocean_Offshore_Quad_Med_Dist_Loop_01_st}
 }
 {id=WavePlayer_16, properties {bLooping=true, soundWaveAssetPtr=/Game/Sounds/Apollo/Biomes/DirtyDocks/Biome_DirtyDocks_Day_Quad}}
 {id=WavePlayer_18, properties {bLooping=true, soundWaveAssetPtr=/Game/Athena/Sounds/Water/Ambient/Amb_Water_Ocean_Offshore_Quad_Loop_01}}
 {id=WavePlayer_3, properties {soundWaveAssetPtr=/Game/Sounds/Apollo/Biomes/Coastal_Areas/Biome_Coastal_Waves_OneShot_01}}
 {id=WavePlayer_6, properties {soundWaveAssetPtr=/Game/Sounds/Apollo/Biomes/Coastal_Areas/Biome_Coastal_Waves_OneShot_02}}
 {id=WavePlayer_7, properties {soundWaveAssetPtr=/Game/Sounds/Apollo/Biomes/Coastal_Areas/Biome_Coastal_Waves_OneShot_03}}
 {id=WavePlayer_8, properties {soundWaveAssetPtr=/Game/Sounds/Apollo/Biomes/Coastal_Areas/Biome_Coastal_Waves_OneShot_04}}
 {id=WavePlayer_9, properties {soundWaveAssetPtr=/Game/Sounds/Apollo/Biomes/Coastal_Areas/Biome_Coastal_Waves_OneShot_05}}
]
properties {
 bOverrideAttenuation=true
 maxDistance=4000
 soundClassObject=&Fort_UI_InGame_Critical
 volumeMultiplier=0.5
 attenuationOverrides {bSpatialize=false, falloffDistance=1500, attenuationShapeExtents {x=2500}}
 concurrencyOverrides {maxCount=1, resolutionRule=PreventNew}
}
